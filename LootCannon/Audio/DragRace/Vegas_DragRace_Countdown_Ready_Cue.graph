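# Sound cue for the "Ready" step of the drag race countdown in the Vegas LootCannon content.
# A random node picks among three Ready wave variations with equal weights.
# Playback is boosted in volume, routed through the quests sound class, and spatialized with the gadget-activate attenuation settings.
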SoundCue /LootCannon/Audio/DragRace/Vegas_DragRace_Countdown_Ready_Cue
root=Random_0
nodes [
 {id=Random_0, children [WavePlayer_1,WavePlayer_2,WavePlayer_3], properties {weights [1,1,1]}}
 {id=WavePlayer_1, properties {soundWaveAssetPtr=&/Vegas_DragRacing_Countdown_Ready_01}}
 {id=WavePlayer_2, properties {soundWaveAssetPtr=&/Vegas_DragRacing_Countdown_Ready_02}}
 {id=WavePlayer_3, properties {soundWaveAssetPtr=&/Vegas_DragRacing_Countdown_Ready_03}}
]
properties {
 attenuationSettings=&Fort_GadgetActivate_Att
 maxDistance=4000
 soundClassObject=&Fort_Quests
 volumeMultiplier=1.5
}
