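# Sound cue for the countdown warning in the Vegas drag race of LootCannon.
# A single wave player plays the countdown warning sound wave directly as the root, with no randomization or mixing.
# It uses the gadget-activate attenuation preset and the quests sound class, with a raised volume and a set maximum audible distance.
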SoundCue /LootCannon/Audio/DragRace/Vegas_DragRace_Countdown_Warning_Cue
root=WavePlayer_1
nodes [{id=WavePlayer_1, properties {soundWaveAssetPtr=&/Vegas_DragRacing_Countdown_Warning}}]
properties {
 attenuationSettings=&Fort_GadgetActivate_Att
 maxDistance=4000
 soundClassObject=&Fort_Quests
 volumeMultiplier=1.2
}
