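# Sound cue for a player entering the end zone of the Vegas drag race in LootCannon.
# It plays a single wave directly from the root wave player, with no randomization or modulation.
# It is routed to the in-game SFX sound class, boosted in volume, and spatialized with the drag race stage-explosion attenuation.
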SoundCue /LootCannon/Audio/DragRace/Vegas_DragRace_EndZone_PlayerEnter_Cue
root=WavePlayer_2
nodes [{id=WavePlayer_2, properties {soundWaveAssetPtr=&/Vegas_DragRacing_EndZone_PlayerEnter}}]
properties {
 attenuationSettings=&/ATT_Vegas_DragRace_StageExplo
 maxDistance=5500
 soundClassObject=&Fort_SFX_InGame
 volumeMultiplier=1.6
}
