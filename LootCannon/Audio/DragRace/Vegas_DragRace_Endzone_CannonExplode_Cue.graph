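# Sound cue for the cannon explosion at the end zone of the Vegas drag race in LootCannon.
# A single wave player feeds the cue output directly, with no randomization or mixing.
# It plays boosted in volume through the in-game SFX class, with drag race stage explosion attenuation limiting its range.
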SoundCue /LootCannon/Audio/DragRace/Vegas_DragRace_Endzone_CannonExplode_Cue
root=WavePlayer_3
nodes [{id=WavePlayer_3, properties {soundWaveAssetPtr=&/Vegas_DragRacing_Endzone_CannonExplode}}]
properties {
 attenuationSettings=&/ATT_Vegas_DragRace_StageExplo
 maxDistance=5500
 soundClassObject=&Fort_SFX_InGame
 volumeMultiplier=1.5
}
