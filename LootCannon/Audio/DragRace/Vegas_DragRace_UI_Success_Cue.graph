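# Sound cue for the success sound of the Vegas drag race, named as a UI sound, in the LootCannon content.
# A single wave player feeds the success wave straight to the output, with boosted volume and the Fort_Quests sound class.
# The cue overrides attenuation with its own falloff and non-spatialized radius, so the referenced Fort_GadgetActivate_Att asset is bypassed and the sound still fades with distance.
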SoundCue /LootCannon/Audio/DragRace/Vegas_DragRace_UI_Success_Cue
root=WavePlayer_2
nodes [{id=WavePlayer_2, properties {soundWaveAssetPtr=&/Vegas_DragRacing_UI_Success}}]
properties {
 attenuationSettings=&Fort_GadgetActivate_Att
 bOverrideAttenuation=true
 maxDistance=5500
 soundClassObject=&Fort_Quests
 volumeMultiplier=1.7
 attenuationOverrides {falloffDistance=5000, nonSpatializedRadiusStart=1024, attenuationShapeExtents {x=500}}
}
